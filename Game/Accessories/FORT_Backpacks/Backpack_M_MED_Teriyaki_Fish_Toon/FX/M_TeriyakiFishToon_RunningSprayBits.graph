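# Masked unlit material for Niagara mesh and sprite particles, used for running spray bits on the Teriyaki Fish Toon backpack.
# Base color blends from black to particle color by UseParticleColor.
# A sphere mask on interpolated camera-space vertex normals paints a fake toon highlight, then feeds MF_FX_EmissiveAdjustments.
# Panning blurred noise, offset per particle and scaled by the VS_Wobble dynamic parameter, multiplies the vertex normal to produce a wobble offset.
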Material /Game/Accessories/FORT_Backpacks/Backpack_M_MED_Teriyaki_Fish_Toon/FX/M_TeriyakiFishToon_RunningSprayBits
graph {
 properties {
  bUsedWithNiagaraMeshParticles=true
  bUsedWithNiagaraSprites=true
  blendMode=BLEND_Masked
  pixelDepthOffsetMode=PDOM_Legacy
  shadingModel=MSM_Unlit
  usedShadingModels=Unlit
 }
 nodes [
  n1 = Constant3Vector()
  n2 = ParticleColor()
  n3 = ScalarParameter() {parameterName=UseParticleColor}
  n4 = Transform(None=VertexNormalWS()) {transformSourceType=TRANSFORMSOURCE_World, transformType=TRANSFORM_Camera}
  n5 = SphereMask(A=ComponentMask(None=VertexInterpolator(VS=@n4).PS) {g=true, r=true}, B=Constant2Vector() {g=0.35, r=0.35}) {attenuationRadius=0.2, hardnessPercent=90}
  MaterialFunctionCall(Emissive=LinearInterpolate(A=LinearInterpolate(A=@n1, B=@n2.RGB, Alpha=@n3), Alpha=Saturate(None=@n5))) {materialFunction=&MF_FX_EmissiveAdjustments}
  n6 = Add(A=TextureCoordinate() {uTiling=0.8, vTiling=0.8}, B=ParticleRandom())
  n7 = Multiply(A=TextureSample(UVs=Panner(Coordinate=@n6) {speedY=-1}) {samplerType=SAMPLERTYPE_Grayscale, texture=&LowResBlurredNoise}.R) {constB=2}
  Multiply(
   A=VertexNormalWS()
   B=Multiply(A=Add(A=@n7) {constB=-1}, B=DynamicParameter() {paramNames [VS_Wobble,Param2,Param3,Param4]}.VS_Wobble) {constB=2.5}
  )
 ]
}
